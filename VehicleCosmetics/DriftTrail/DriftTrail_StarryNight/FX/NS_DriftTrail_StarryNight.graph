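# Niagara system for the StarryNight vehicle drift trail cosmetic, combining ribbons, sprites and wheel-wrap meshes.
# Two ribbons follow the car through three renderers, with game-mode-dependent lifetime, separate width curves and an animated sine/cosine lateral offset.
# Drifting, spinning sprites with curl noise and two looping mesh wraps sample wheel sockets on the skeletal mesh.
# All emitters apply painted-variant color overrides, age-based gradients on the ribbons, and per-wheel tire FX masking.
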NiagaraSystem /VehicleCosmetics/DriftTrail/DriftTrail_StarryNight/FX/NS_DriftTrail_StarryNight
dependencies {
 usedDataInterfaces [
  &NiagaraDataInterfaceColorCurve
  &NiagaraDataInterfaceCurve
  &NiagaraDataInterfaceCamera
  &NiagaraDataInterfaceSkeletalMesh
 ]
 usedDynamicInputs []
 usedModules [
  &self:SystemSpawnScript.SystemScriptSource.SystemScriptGraph.NiagaraNodeAssignment_0.SetVariables_#1
  &FN_UnpackVehicleFlags
  &SystemState
  /Game/Effects/Niagara/Modules/Vehicle/FN_UpdateVehicleAttribs
  /Game/Effects/Niagara/Modules/Vehicle/FN_VehicleFlags_Cheatcodes
 ]
}
emitters {
 NE_VC_DriftTrail_Ribbon_Parent002 {
  emitterUpdateScript [
   &EmitterState(Life Cycle Mode=::NewEnumerator1, Loop Duration=1, UseLoopDelay=false)
   &VC_RateQualityHandler(High Spawn Rate=60, Low Spawn Rate=60, Medium Spawn Rate=60, Use Custom Rate?=true)
   &VC_SpawnRibbon_v1(Number of Ribbons=2, Spawn Time=-1, SpawnRate=$Emitter.Spawn Rate)
  ]
  particleSpawnScript [
   &VC_PaintedVariantsOverride() {enabled=false}
   &VC_InitializeRibbon_v1(
    Color=$Particles.Color
    Custom Offset=true
    Enable Skeletal Position=true
    Lifetime=&MergeHelper_FloatSwitch(0=0.8, 1=0.6, 2=0.6, 3=?, Index=$User.GameModeID, Num=::NewEnumerator1)
    Local Vector Offset=&AddVectorToPosition(Position=<-30 0 -35>, Vector=Linked() {linkedVariable {name=User.DTCarOffset, type=NiagaraPosition}})
    Offset Amount=1
    Ribbon Width=55
   )
   /SetVariables(
    Particles.R1_DarkBlueColor:LinearColor=&MultiplyLinearColor(
     Linear Color=&MakeCustomLinearColorFromBool(Bool=$User.bIsPainted, True Color=$User.PaintedColor, False Color=<rgba 1 1 1 1>)
     Scale Factor=<1 1 1 1>
    )
    Particles.R2_YB_Gradient:LinearColor=&MultiplyLinearColor(
     Linear Color=&ColorFromCurve(ColorCurve=DataInterface() {propertyValues {
        AlphaCurve {keys [<at 0.0534 0 RCIM_Cubic>,<at 0.1847 3>,<at 0.3787 3>,<at 1 0 RCIM_Cubic>]}
        BlueCurve {keys [
          <at 0.0572 0.0152>
          {arriveTangent=1.348, interpMode=RCIM_Cubic, leaveTangent=1.348, time=0.195, value=0.0474}
          <at 0.3251 0.3763>
          <at 0.4778 2 RCIM_Cubic>
         ]}
        GreenCurve {keys [
          <at 0.0572 0.4342>
          {arriveTangent=-0.3972, interpMode=RCIM_Cubic, leaveTangent=-0.3972, time=0.195, value=0.6149}
          <at 0.3251 0.3278>
          <at 0.4778 0.9468 RCIM_Cubic>
         ]}
        RedCurve {keys [
          <at 0.0572 0.8796>
          {arriveTangent=-2.7019, interpMode=RCIM_Cubic, leaveTangent=-2.7019, time=0.195, value=0.7969}
          <at 0.3251 0.1559>
          <at 0.4778 0.3304 RCIM_Cubic>
         ]}
       }})
     Scale Factor=&MakeVector4FromVectorAndFloat(
      "Vector (RGB)"=<1 1 1>
      "Float (Alpha)"=&FloatFromCurve(
       FloatCurve=DataInterface() {propertyValues {Curve {keys [
           {interpMode=RCIM_Cubic, tangentMode=RCTM_Break, time=0.1409, value=1.1187}
           {arriveTangent=-2.1973, interpMode=RCIM_Cubic, leaveTangent=-2.1973, time=0.3779, value=1.079}
           {interpMode=RCIM_Cubic, tangentMode=RCTM_Break, time=0.65, value=0}
          ]}}}
       Scale Curve=2
      )
     )
    )
   ) {enabled=false, isSetParameters=true}
   &AddVelocity(Velocity=<-1000 0 25>) {module=AddVelocity001}
  ]
  particleUpdateScript [
   &VC_PaintedVariantsOverride(
    R1 Black:LinearColor=&ColorFromCurve(ColorCurve=DataInterface() {propertyValues {
       AlphaCurve {keys [<at 0 4 RCIM_Cubic>,<at 0.1045 0.9695>,<at 1 0 RCIM_Cubic>]}
       BlueCurve {keys [<at 0.1508 0.1565 RCIM_Cubic>,<at 0.2311 0.0208>]}
       GreenCurve {keys [<at 0.1508 0.1565 RCIM_Cubic>,<at 0.2311 0.0208>]}
       RedCurve {keys [<at 0.1508 0.1565 RCIM_Cubic>,<at 0.2311 0.0208>]}
      }})
    R1 BurntSienna:LinearColor=&ColorFromCurve(ColorCurve=DataInterface() {propertyValues {
       AlphaCurve {keys [<at 0 4 RCIM_Cubic>,<at 0.2166 2>,<at 1 0 RCIM_Cubic>]}
       BlueCurve {keys [<at 0.1183 0.2086>,<at 0.2506 0.0061>]}
       GreenCurve {keys [<at 0.1183 0.3968>,<at 0.2506 0.0193>]}
       RedCurve {keys [<at 0.1183 0.9387>,<at 0.2506 0.067>]}
      }})
    R1 Cobalt:LinearColor=&ColorFromCurve(ColorCurve=DataInterface() {propertyValues {
       AlphaCurve {keys [<at 0.1551 4 RCIM_Cubic>,<at 0.2767 1>,<at 1 0 RCIM_Cubic>]}
       BlueCurve {keys [<at 0.1226 1.3 RCIM_Cubic>,<at 0.2484 0.4531>]}
       GreenCurve {keys [<at 0.1226 0.6773 RCIM_Cubic>,<at 0.2484 0.01>]}
       RedCurve {keys [<at 0.1226 0.3304 RCIM_Cubic>,<at 0.2484 0.0094>]}
      }})
    R1 Crimson:LinearColor=&ColorFromCurve(ColorCurve=DataInterface() {propertyValues {
       AlphaCurve {keys [<at 0 4 RCIM_Cubic>,<at 1 0 RCIM_Cubic>]}
       BlueCurve {keys [<at 0.1486 0.1812>,<at 0.2419 0.0212>]}
       GreenCurve {keys [<at 0.1486 0.162>,<at 0.2419 0.0091>]}
       RedCurve {keys [<at 0.1486 1>,<at 0.2419 0.3864>]}
      }})
    R1 DefaultColor=&ColorFromCurve(ColorCurve=DataInterface() {propertyValues {
       AlphaCurve {keys [
         <at 0 5>
         <at 0.0851 2>
         {arriveTangent=-1.1995, interpMode=RCIM_Cubic, leaveTangent=-1.1995, time=0.2252, value=4}
         <at 0.9208 0.9976>
         <at 1 0 RCIM_Cubic>
        ]}
       BlueCurve {keys [<at 0.0079 0.1026>,<at 0.1092 0.0205>,<at 0.1565 0.2307>,<at 0.2286 0.552>,<at 0.384 0.3095>,<at 0.6099 0.0666>]}
       GreenCurve {keys [<at 0.0079 3.37>,<at 0.1092 0.674>,<at 0.1565 0.4452>,<at 0.2286 0.0931>,<at 0.384 0.0561>,<at 0.6099 0.0194>]}
       RedCurve {keys [<at 0.0079 5>,<at 0.1092 1>,<at 0.1565 0.624>,<at 0.2286 0.0437>,<at 0.384 0.0252>,<at 0.6099 0.006>]}
      }})
    R1 ForestGreen:LinearColor=&ColorFromCurve(ColorCurve=DataInterface() {propertyValues {
       AlphaCurve {keys [<at 0 4 RCIM_Cubic>,<at 1 0 RCIM_Cubic>]}
       BlueCurve {keys [<at 0.1465 0.1651>,<at 0.255 0.0262>]}
       GreenCurve {keys [<at 0.1465 0.7011>,<at 0.255 0.147>]}
       RedCurve {keys [<at 0.1465 0.0091>,<at 0.255 0.0018>]}
      }})
    R1 Grey:LinearColor=&ColorFromCurve(ColorCurve=DataInterface() {propertyValues {
       AlphaCurve {keys [<at 0 1 RCIM_Cubic>,<at 0.8868 0.9>,<at 0.9795 0 RCIM_Cubic>]}
       BlueCurve {keys [<at 0.1313 0.5972>,<at 0.255 0.0666>]}
       GreenCurve {keys [<at 0.1313 0.5972>,<at 0.255 0.0545>]}
       RedCurve {keys [<at 0.1313 0.5776>,<at 0.255 0.0513>]}
      }})
    R1 Lime:LinearColor=&ColorFromCurve(ColorCurve=DataInterface() {propertyValues {
       AlphaCurve {keys [<at 0 1 RCIM_Cubic>,<at 0.8997 0.9>,<at 1 0 RCIM_Cubic>]}
       BlueCurve {keys [<at 0.1421 0.1144>,<at 0.2463 0.0242>]}
       GreenCurve {keys [<at 0.1421 0.9131>,<at 0.2463 0.2874>]}
       RedCurve {keys [<at 0.1421 0.5583>,<at 0.2463 0.1329>]}
      }})
    R1 Orange:LinearColor=&ColorFromCurve(ColorCurve=DataInterface() {propertyValues {
       AlphaCurve {keys [<at 0 1 RCIM_Cubic>,<at 0.9213 0.986>,<at 1 0 RCIM_Cubic>]}
       BlueCurve {keys [<at 0.1465 0.0542 RCIM_Cubic>,<at 0.2419 0.0088>]}
       GreenCurve {keys [<at 0.1465 0.58 RCIM_Cubic>,<at 0.2419 0.1376>]}
       RedCurve {keys [<at 0.1465 1 RCIM_Cubic>,<at 0.2419 0.5469>]}
      }})
    R1 Pink:LinearColor=&ColorFromCurve(ColorCurve=DataInterface() {propertyValues {
       AlphaCurve {keys [<at 0 1 RCIM_Cubic>,<at 0.9277 0.9949>,<at 1 0 RCIM_Cubic>]}
       BlueCurve {keys [<at 0.1486 0.8714>,<at 0.2484 0.2542>]}
       GreenCurve {keys [<at 0.1486 0.2705>,<at 0.2484 0.011>]}
       RedCurve {keys [<at 0.1486 0.9647>,<at 0.2484 0.4342>]}
      }})
    R1 Purple:LinearColor=&ColorFromCurve(ColorCurve=DataInterface() {propertyValues {
       AlphaCurve {keys [<at 0 1 RCIM_Cubic>,<at 0.8976 0.9795>,<at 1 0 RCIM_Cubic>]}
       BlueCurve {keys [<at 0.153 1>,<at 0.255 0.305>]}
       GreenCurve {keys [<at 0.153 0.1798>,<at 0.255 0.0052>]}
       RedCurve {keys [<at 0.153 0.609>,<at 0.255 0.1329>]}
      }})
    R1 Saffron:LinearColor=&ColorFromCurve(ColorCurve=DataInterface() {propertyValues {
       AlphaCurve {keys [<at 0 1 RCIM_Cubic>,<at 0.9083 0.9979>,<at 1 0 RCIM_Cubic>]}
       BlueCurve {keys [<at 0.1421 0.1301>,<at 0.2506 0.0273>]}
       GreenCurve {keys [<at 0.1421 0.7454>,<at 0.2506 0.3231>]}
       RedCurve {keys [<at 0.1421 0.8796>,<at 0.2506 0.4179>]}
      }})
    R1 SkyBlue:LinearColor=&ColorFromCurve(ColorCurve=DataInterface() {propertyValues {
       AlphaCurve {keys [<at 0 1 RCIM_Cubic>,<at 0.9428 0.9815>,<at 1 0 RCIM_Cubic>]}
       BlueCurve {keys [<at 0.1031 1>,<at 0.2484 0.3185>]}
       GreenCurve {keys [<at 0.1031 0.9216>,<at 0.2484 0.2307>]}
       RedCurve {keys [<at 0.1031 0.0685>,<at 0.2484 0.0012>]}
      }})
    R1 TitaniumWhite:LinearColor=&ColorFromCurve(ColorCurve=DataInterface() {propertyValues {
       AlphaCurve {keys [<at 0 1 RCIM_Cubic>,<at 1 0 RCIM_Cubic>]}
       BlueCurve {keys [<at 0.1421 0.855>,<at 0.2463 0.1714>]}
       GreenCurve {keys [<at 0.1421 0.8148>,<at 0.2463 0.1683>]}
       RedCurve {keys [<at 0.1421 0.855>,<at 0.2463 0.159>]}
      }})
    R2 Black:LinearColor=&ColorFromCurve(ColorCurve=DataInterface() {propertyValues {
       AlphaCurve {keys [
         <at 0 0>
         {arriveTangent=3.3444, interpMode=RCIM_Cubic, leaveTangent=3.3444, time=0.0636, value=1}
         <at 0.3588 1.2>
         <at 0.5522 0 RCIM_Cubic>
        ]}
       BlueCurve {keys [<at 0.0463 0.1354>,<at 0.4752 0.7188>]}
       GreenCurve {keys [<at 0.0463 0.1354>,<at 0.4752 0.7188>]}
       RedCurve {keys [<at 0.0463 0.1354>,<at 0.4752 0.7188>]}
      }})
    R2 BurntSienna:LinearColor=&ColorFromCurve(ColorCurve=DataInterface() {propertyValues {
       AlphaCurve {keys [<at 0 0>, {arriveTangent=5.3491, interpMode=RCIM_Cubic, leaveTangent=5.3491, time=0.0442, value=3}, <at 0.3739 2>, <at 0.5566 0 RCIM_Cubic>]}
       BlueCurve {keys [<at 0.0248 0.115 RCIM_Cubic>,<at 0.3868 0.2014>]}
       GreenCurve {keys [<at 0.0248 0.1634 RCIM_Cubic>,<at 0.3868 0.2311>]}
       RedCurve {keys [<at 0.0248 0.3391 RCIM_Cubic>,<at 0.3868 0.3391>]}
      }})
    R2 Cobalt:LinearColor=&ColorFromCurve(ColorCurve=DataInterface() {propertyValues {
       AlphaCurve {keys [<at 0 0 RCIM_Cubic>,<at 0.0356 3>,<at 0.3761 2.5>,<at 0.5414 0 RCIM_Cubic>]}
       BlueCurve {keys [<at 0.0894 0.88>,<at 0.4256 3>]}
       GreenCurve {keys [<at 0.0894 0.0245>,<at 0.4256 0.1875>]}
       RedCurve {keys [<at 0.0894 0.0176>,<at 0.4256 0.1875>]}
      }})
    R2 Crimson:LinearColor=&ColorFromCurve(ColorCurve=DataInterface() {propertyValues {
       AlphaCurve {keys [
         <at 0 0>
         {arriveTangent=8.9581, interpMode=RCIM_Cubic, leaveTangent=8.9581, time=0.0442, value=1}
         <at 0.2791 2.5>
         <at 0.5435 0 RCIM_Cubic>
        ]}
       BlueCurve {keys [<at 0 0.0405 RCIM_Cubic>,<at 0.3427 0.2462>,<at 0.4881 0.3841>]}
       GreenCurve {keys [<at 0 0.0176 RCIM_Cubic>,<at 0.3427 0.2307>,<at 0.4881 0.34>]}
       RedCurve {keys [<at 0 0.88 RCIM_Cubic>,<at 0.3427 0.855>,<at 0.4881 2>]}
      }})
    R2 DefaultColor:LinearColor=&ColorFromCurve(ColorCurve=DataInterface() {propertyValues {
       AlphaCurve {keys [<at 0.0191 0 RCIM_Cubic>,<at 0.1847 3>,<at 0.3787 3>,<at 0.5571 0 RCIM_Cubic>]}
       BlueCurve {keys [<at 0.1835 0.0152>, {arriveTangent=-0.5504, interpMode=RCIM_Cubic, time=0.3299, value=2}]}
       GreenCurve {keys [<at 0.1835 0.4342>, {arriveTangent=1.5922, interpMode=RCIM_Cubic, time=0.3299, value=0.9468}]}
       RedCurve {keys [<at 0.1835 0.8796>, {arriveTangent=3.3953, interpMode=RCIM_Cubic, time=0.3299, value=0.3304}]}
      }})
    R2 ForestGreen:LinearColor=&ColorFromCurve(ColorCurve=DataInterface() {propertyValues {
       AlphaCurve {keys [<at 0 0>, {arriveTangent=6.0856, interpMode=RCIM_Cubic, leaveTangent=6.0856, time=0.0485, value=2}, <at 0.3286 2>, <at 0.5435 0 RCIM_Cubic>]}
       BlueCurve {keys [<at 0.0463 0.0608 RCIM_Cubic>,<at 0.4493 0.4231>]}
       GreenCurve {keys [<at 0.0463 0.88 RCIM_Cubic>,<at 0.4493 2>]}
       RedCurve {keys [<at 0.0463 0.0176 RCIM_Cubic>,<at 0.4493 0.34>]}
      }})
    R2 Grey:LinearColor=&ColorFromCurve(ColorCurve=DataInterface() {propertyValues {
       AlphaCurve {keys [
         <at 0.0119 0>
         {arriveTangent=2.9746, interpMode=RCIM_Cubic, leaveTangent=2.9746, time=0.0528, value=1}
         <at 0.348 1>
         <at 0.5414 0 RCIM_Cubic>
        ]}
       BlueCurve {keys [<at 0 0.4167 RCIM_Cubic>,<at 0.4429 2>]}
       GreenCurve {keys [<at 0 0.4167 RCIM_Cubic>,<at 0.4429 2>]}
       RedCurve {keys [<at 0 0.4167 RCIM_Cubic>,<at 0.4429 2>]}
      }})
    R2 Lime:LinearColor=&ColorFromCurve(ColorCurve=DataInterface() {propertyValues {
       AlphaCurve {keys [
         <at 0.0183 0>
         {arriveTangent=6.9779, interpMode=RCIM_Cubic, leaveTangent=6.9779, time=0.0593, value=1}
         <at 0.3049 2>
         <at 0.5457 0 RCIM_Cubic>
        ]}
       BlueCurve {keys [<at 0.083 0.015 RCIM_Cubic>,<at 0.4924 0.34>]}
       GreenCurve {keys [<at 0.083 0.88 RCIM_Cubic>,<at 0.4924 2>]}
       RedCurve {keys [<at 0.083 0.4487 RCIM_Cubic>,<at 0.4924 1.1723>]}
      }})
    R2 Orange:LinearColor=&ColorFromCurve(ColorCurve=DataInterface() {propertyValues {
       AlphaCurve {keys [
         <at 0.0205 0>
         {arriveTangent=5.5907, interpMode=RCIM_Cubic, leaveTangent=5.5907, time=0.1261, value=2.5}
         <at 0.3782 2>
         <at 0.5435 0 RCIM_Cubic>
        ]}
       BlueCurve {keys [<at 0.0657 0.015 RCIM_Cubic>,<at 0.4881 0.165>]}
       GreenCurve {keys [<at 0.0657 0.3167 RCIM_Cubic>,<at 0.4881 0.4563>]}
       RedCurve {keys [<at 0.0657 0.88 RCIM_Cubic>,<at 0.4881 1>]}
      }})
    R2 Pink:LinearColor=&ColorFromCurve(ColorCurve=DataInterface() {propertyValues {
       AlphaCurve {keys [
         <at 0.0183 0>
         {arriveTangent=4.1431, interpMode=RCIM_Cubic, leaveTangent=4.1431, time=0.083, value=1.5}
         <at 0.3804 1.5>
         <at 0.5349 0 RCIM_Cubic>
        ]}
       BlueCurve {keys [<at 0.0614 0.5654 RCIM_Cubic>,<at 0.5118 1.3913>]}
       GreenCurve {keys [<at 0.0614 0.015 RCIM_Cubic>,<at 0.5118 0.33>]}
       RedCurve {keys [<at 0.0614 0.88 RCIM_Cubic>,<at 0.5118 2>]}
      }})
    R2 Purple:LinearColor=&ColorFromCurve(ColorCurve=DataInterface() {propertyValues {
       AlphaCurve {keys [<at 0 0>, {arriveTangent=5.2581, interpMode=RCIM_Cubic, leaveTangent=5.2581, time=0.0679, value=1}, <at 0.3804 2>, <at 0.5262 0 RCIM_Cubic>]}
       BlueCurve {keys [<at 0.0528 0.88 RCIM_Cubic>,<at 0.6002 2>]}
       GreenCurve {keys [<at 0.0528 0.015 RCIM_Cubic>,<at 0.6002 0.1562>]}
       RedCurve {keys [<at 0.0528 0.37 RCIM_Cubic>,<at 0.6002 0.913>]}
      }})
    R2 Saffron:LinearColor=&ColorFromCurve(ColorCurve=DataInterface() {propertyValues {
       AlphaCurve {keys [<at 0.0291 0>,<at 0.0959 1.5>,<at 0.4601 1.5>,<at 0.57 0 RCIM_Cubic>]}
       BlueCurve {keys [<at 0.07 0.015>,<at 0.5054 0.33>]}
       GreenCurve {keys [<at 0.07 0.6777>,<at 0.5054 1.6095>]}
       RedCurve {keys [<at 0.07 0.88>,<at 0.5054 2>]}
      }})
    R2 SkyBlue:LinearColor=&ColorFromCurve(ColorCurve=DataInterface() {propertyValues {
       AlphaCurve {keys [<at 0 0 RCIM_Cubic>,<at 0.0959 1.9879>,<at 0.3157 1.995>,<at 0.4321 2>,<at 0.5479 0 RCIM_Cubic>]}
       BlueCurve {keys [<at 0.0916 0.88 RCIM_Cubic>,<at 0.5334 2>]}
       GreenCurve {keys [<at 0.0916 0.5931 RCIM_Cubic>,<at 0.5334 1.4461>]}
       RedCurve {keys [<at 0.0916 0.015 RCIM_Cubic>,<at 0.5334 0.33>]}
      }})
    R2 TitaniumWhite:LinearColor=&ColorFromCurve(ColorCurve=DataInterface() {propertyValues {
       AlphaCurve {keys [<at 0 0 RCIM_Cubic>,<at 0.0679 1.4>,<at 0.3653 1.5>,<at 0.5544 0 RCIM_Cubic>]}
       BlueCurve {keys [<at 0.0463 0.5573 RCIM_Cubic>,<at 0.4816 2>]}
       GreenCurve {keys [<at 0.0463 0.5573 RCIM_Cubic>,<at 0.4816 2>]}
       RedCurve {keys [<at 0.0463 0.5573 RCIM_Cubic>,<at 0.4816 2>]}
      }})
    R3 Black:LinearColor=<rgba 0.0052 0.0052 0.0052 1>
    R3 BurntSienna:LinearColor=<rgba 0.0312 0.0096 0.0039 1>
    R3 Cobalt:LinearColor=<rgba 0 0.004 0.2135 1>
    R3 Crimson:LinearColor=<rgba 0.1458 0.0114 0.015 1>
    R3 DefaultAlpha=4
    R3 DefaultColor:LinearColor=<rgba 0.0011 0.0034 0.0208 1>
    R3 ForestGreen:LinearColor=<rgba 0 0.0781 0.0039 1>
    R3 Grey:LinearColor=<rgba 0.1042 0.1042 0.1042 1>
    R3 Lime:LinearColor=<rgba 0.0677 0.1354 0 1>
    R3 Orange:LinearColor=<rgba 0.3177 0.1108 0 1>
    R3 Pink:LinearColor=<rgba 0.2344 0.0281 0.1594 1>
    R3 Purple:LinearColor=<rgba 0.062 0 0.151 1>
    R3 Saffron:LinearColor=<rgba 0.2188 0.1677 0 1>
    R3 SkyBlue:LinearColor=<rgba 0 0.1213 0.1823 1>
    R3 TitaniumWhite:LinearColor=<rgba 0.1406 0.1406 0.1406 1>
    Use R1 Painted Overrides=true
    Use R2 Painted Overrides=true
    Use R3 Painted Overrides=true
   ) {module=VC_PaintedVariantsOverride001}
   &FN_DisableTireFX(Mask Index Attribute:NiagaraInt32=$Particles.SpawnGroup, Mask Index FL=2, Mask Index FR=3, Mask Index RL=0, Mask Index RR=1)
   &InheritSourceMovement(
    Applied Position Scale=&Multiply_VectorByFloat(Vector=<1 1 1>, Float=1)
    ApplyForce=false
    ApplyPosition=true
    ApplyPositionFalloff=false
   )
   &ParticleState()
   &DynamicMaterialParameters(
    Index 0 Param 4=10.4934
    Module.Index0Float0Write=true
    Module.Index0Float1Write=true
    Module.Index0Float2Write=true
    Module.Index0Float3Write=true
   )
   /SetVariables(Particles.DotProduct:NiagaraFloat=&DotProduct(
     Vector A=/Niagara/DynamicInputs/Math/TransformVectorByMatrix(
      Vector:Vector3f=<1 0 0>
      Transformation Matrix:NiagaraMatrix=$Engine.Owner.SystemLocalToWorld
      Transform Vector as Position:NiagaraBool=false
     )
     Vector B=&NormalizeVector(
      Vector To Normalize=&SubtractVectorFromPosition(
       Position=$Engine.Owner.Position
       Vector=&ReturnCameraPosition(Coordinate Space:ENiagaraCoordinateSpace=::Simulation, Camera Query:NiagaraDataInterfaceCamera=DataInterface())
      )
      Safe Normalize=true
      Safe Fallback Threshold=0
      Safe Fallback Vector=<0 0 0.001>
     )
    )) {isSetParameters=true}
   /SetVariables(Particles.CustomPosition_R1:NiagaraPosition=&AddVectorToPosition(
     Position=$Particles.Position
     Vector=&MakeVector(
      X=0
      Y=&MaskFloat(
       Mask Source=0
       Mask Parameter=?
       Mask=0
       Masked Float=&Cosine(
        Normalized Angle=&Add_Float(A=0, B=&Subtract_Float(A=$Emitter.NormalizedLoopAge, B=$Particles.NormalizedAge))
        Period=1
        Scale=&FloatFromCurve(
         FloatCurve=DataInterface() {propertyValues {Curve {keys [
             <at 0 0 RCIM_Cubic>
             {arriveTangent=4.1101, interpMode=RCIM_Cubic, leaveTangent=4.1101, tangentMode=RCTM_User, time=0.158, value=0.71}
             {arriveTangent=3.9851, interpMode=RCIM_Cubic, leaveTangent=3.9851, time=0.4103, value=3.8}
             {
              arriveTangent=-0.5881
              arriveTangentWeight=0.5973
              interpMode=RCIM_Cubic
              leaveTangent=-0.5881
              leaveTangentWeight=0.5973
              tangentMode=RCTM_User
              time=0.7585
              value=3.103
             }
            ]}}}
         Scale Curve=&Lerp_Float(
          A=10
          B=10
          Alpha=&NormalizeFloat(Value=&Power(a=&Add_Float(A=$Particles.DotProduct, B=0.1), b=5), Min=0, Max=1)
         )
        )
        Bias=0
       )
       Passthrough Float=&Sine(
        Normalized Angle=&Add_Float(A=0, B=&Subtract_Float(A=$Emitter.NormalizedLoopAge, B=$Particles.NormalizedAge))
        Period=1
        Scale=&FloatFromCurve(
         FloatCurve=DataInterface() {propertyValues {Curve {keys [
             <at 0 0 RCIM_Cubic>
             {arriveTangent=4.1101, interpMode=RCIM_Cubic, leaveTangent=4.1101, tangentMode=RCTM_User, time=0.1316, value=0.71}
             {arriveTangent=3.1742, interpMode=RCIM_Cubic, leaveTangent=3.1742, time=0.4103, value=3.8}
             {
              arriveTangent=-0.5881
              arriveTangentWeight=0.5973
              interpMode=RCIM_Cubic
              leaveTangent=-0.5881
              leaveTangentWeight=0.5973
              tangentMode=RCTM_User
              time=0.7585
              value=2.7
             }
            ]}}}
         Scale Curve=&Lerp_Float(
          A=10
          B=10
          Alpha=&NormalizeFloat(Value=&Power(a=&Add_Float(A=$Particles.DotProduct, B=0.1), b=5), Min=0, Max=1)
         )
        )
        Bias=0
       )
      )
      Z=0
     )
    )) {isSetParameters=true}
   /SetVariables(
    Particles.R2_Width:NiagaraFloat=&FloatFromCurve(
     FloatCurve=DataInterface() {propertyValues {Curve {keys [
         {arriveTangent=3.2862, interpMode=RCIM_Cubic, leaveTangent=3.2862, tangentMode=RCTM_User, time=-0.0016, value=0.4072}
         {arriveTangent=0.2431, interpMode=RCIM_Cubic, leaveTangent=0.2431, time=0.1918, value=0.8579}
         <at 0.9611 0.6412>
        ]}}}
     Scale Curve=20
    )
    Particles.R3_Width:NiagaraFloat=&FloatFromCurve(
     FloatCurve=DataInterface() {propertyValues {Curve {keys [
         {arriveTangent=3.2862, interpMode=RCIM_Cubic, leaveTangent=3.2862, tangentMode=RCTM_User, time=0.0007, value=0.2288}
         {arriveTangent=0.4293, interpMode=RCIM_Cubic, leaveTangent=0.4293, time=0.1846, value=0.835}
         <at 0.9611 0.6412>
        ]}}}
     Scale Curve=60
    )
   ) {isSetParameters=true}
   &ScaleRibbonWidth(Ribbon Width Scale=&FloatFromCurve(FloatCurve=DataInterface() {propertyValues {Curve {keys [
         {arriveTangent=3.2862, interpMode=RCIM_Cubic, leaveTangent=3.2862, tangentMode=RCTM_User, time=0.0007, value=0.2288}
         {arriveTangent=0.4293, interpMode=RCIM_Cubic, leaveTangent=0.4293, time=0.1846, value=0.835}
         <at 0.9611 0.6412>
        ]}}}))
   &SolveForcesAndVelocity(Clamp Velocity=false, Limit Acceleration=false)
   &VC_UpdateRibbon_v1(Enable Skeletal Position=true)
  ]
  rendererClasses [&NiagaraRibbonRendererProperties]
  renderers [
   {rendererClass=&NiagaraRibbonRendererProperties}
   {rendererClass=&NiagaraRibbonRendererProperties, rendererIndex=1}
   {rendererClass=&NiagaraRibbonRendererProperties, rendererIndex=2}
  ]
 }
 NE_VC_DriftTrail_Sprite_Parent {
  emitterUpdateScript [
   &EmitterState()
   &SpawnRate(SpawnRate=8, Use Spawn Probability=false)
  ]
  particleSpawnScript [
   &VC_PaintedVariantsOverride(
    Black=<rgba 0.0938 0.0938 0.0938 1>
    BurntSienna=<rgba 0.2083 0.1305 0.1096 1>
    Cobalt=<rgba 0.0249 0.0398 0.83 1>
    Crimson=<rgba 0.83 0.0249 0.0467 1>
    DefaultColor=<rgba 0.8388 0.4735 0.0185 1>
    ForestGreen=<rgba 0.0185 0.839 0.0599 1>
    Grey=<rgba 0.3565 0.3565 0.3565 1>
    Lime=<rgba 0.4287 0.839 0.0185 1>
    Orange=<rgba 0.839 0.3307 0.0584 1>
    Pink=<rgba 0.839 0.1442 0.5863 1>
    Purple=<rgba 0.3552 0.0185 0.839 1>
    Saffron=<rgba 0.839 0.6857 0.1835 1>
    SkyBlue=<rgba 0.0918 0.589 0.839 1>
    TitaniumWhite=<rgba 1 1 1 1>
   )
   &InitializeParticle(
    Color=$Particles.Color
    Lifetime=&MergeHelper_FloatSwitch(0=1, 1=0.8, 2=0.8, 3=?, Index=&SpawnGroup(), Num=::NewEnumerator1)
    Sprite Rotation Mode=::NewEnumerator2
    Sprite Size Mode=::NewEnumerator1
    Uniform Sprite Size Max=25
    Uniform Sprite Size Min=7
    UsePositionOffset=false
    Write Material Random=true
   )
   &PartitionParticles()
   &SkeletalMeshLocation(
    Bone / Socket Index=&Add_Integer(A=$Particles.PartitionParticles.Partition, B=4)
    Bone / Socket Sampling Mode=::NewEnumerator5
    Interpolate Mesh Sampling=false
    MaskSpawnGroup=false
    Mesh Sampling Type=::NewEnumerator6
    Sampled Position Offset=&AddVectorToPosition(Position=$User.offset, Vector=<-55 0 -25>)
    Skeletal Mesh=$System.SkeletalMesh
   )
   &AddVelocity(Velocity=&VC_Mask_Vectors(
     FrontEnd Value:Vector3f=<-400 0 0>
     blsFrontEndPreview:NiagaraBool=false
     SpeedMin Value:Vector3f=<-200 0 0>
     SpeedMax Value:Vector3f=<-400 0 0>
     Speed Alpha:NiagaraFloat=$System.Speed_Alpha
    ))
   &ShapeLocation(MaskSpawnGroup=false, Override Seed=false, Sphere Radius=15, UseOwnerScale=true)
  ]
  particleUpdateScript [
   &FN_DisableTireFX(
    Mask Index Attribute:NiagaraInt32=$Particles.PartitionParticles.Partition
    Mask Index FL=2
    Mask Index FR=3
    Mask Index RL=0
    Mask Index RR=1
   )
   &InheritSourceMovement(ApplyForce=false, ApplyPosition=true, ApplyPositionFalloff=false)
   &ParticleState()
   &ScaleSpriteSize(Uniform Curve Sprite Scale=DataInterface() {propertyValues {Curve {
       defaultValue=0
       keys [{interpMode=RCIM_Cubic, tangentMode=RCTM_User, time=0.7, value=1}, {interpMode=RCIM_Cubic, tangentMode=RCTM_User, time=1, value=0}]
      }}})
   &Drag(Drag=&RandomRangeFloat(Minimum=0.5, Maximum=1.5), Use Linear Drag=true, Use Rotational Drag=true)
   &CurlNoiseForce(
    Bias Noise=false
    Noise Frequency=&RandomRangeFloat(Minimum=25, Maximum=65)
    Noise Strength=150
    Pan Noise=false
   )
   &SolveForcesAndVelocity(Clamp Velocity=false, Limit Acceleration=false)
   &SpriteRotationRate(Rotation Rate=&RandomRangeFloat(Minimum=65, Maximum=125))
   &ScaleColor(
    Scale Alpha=&FloatFromCurve(FloatCurve=DataInterface() {propertyValues {Curve {keys [{interpMode=RCIM_Cubic, tangentMode=RCTM_Break, time=0.35, value=1}, {interpMode=RCIM_Cubic, tangentMode=RCTM_Break, time=1, value=0}]}}})
    ScaleA=true
    ScaleRGB=true
   )
  ]
  rendererClasses [&NiagaraSpriteRendererProperties]
  renderers [{rendererClass=&NiagaraSpriteRendererProperties}]
 }
 WheelWrap02 {
  emitterUpdateScript [
   &EmitterState(
    Inactive Response=::NewEnumerator1
    Life Cycle Mode=::NewEnumerator1
    Loop Behavior=::NewEnumerator1
    Loop Duration Mode=::NewEnumerator1
    UseLoopDelay=false
   )
   &SpawnBurst_Instantaneous(Spawn Count=2, Use Spawn Probability=false, UseLoopCountLimit=false)
  ]
  particleSpawnScript [
   &VC_PaintedVariantsOverride(
    BurntSienna=<rgba 0.1042 0.0322 0.0128 1>
    DefaultColor=<rgba 1 0.4948 0.0017 1>
    ForestGreen=<rgba 0 0.25 0.0126 1>
    Lime=<rgba 0.2599 0.4583 0.0215 1>
    Orange=<rgba 0.3281 0.1095 0.0085 1>
    Pink=<rgba 0.2917 0.0349 0.1983 1>
    Purple=<rgba 0.1775 0.0382 0.4219 1>
    R1 Black:LinearColor=<rgba 0.0625 0.0625 0.0625 1>
    R1 BurntSienna:LinearColor=<rgba 0.3763 0.1195 0.0612 1>
    R1 Cobalt:LinearColor=<rgba 0.0973 0.2125 0.5573 1>
    R1 Crimson:LinearColor=<rgba 0.7344 0.0841 0.1015 1>
    R1 DefaultColor=<rgba 0.1273 0.6251 0.7552 1>
    R1 ForestGreen:LinearColor=<rgba 0.0277 0.875 0.1448 1>
    R1 Grey:LinearColor=<rgba 0.3281 0.3281 0.3281 1>
    R1 Lime:LinearColor=<rgba 0.7305 0.9734 0.0037 1>
    R1 Orange:LinearColor=<rgba 0.9647 0.5994 0.1105 1>
    R1 Pink:LinearColor=<rgba 0.5052 0.1289 0.3684 1>
    R1 Purple:LinearColor=<rgba 0.3441 0.0518 0.8698 1>
    R1 Saffron:LinearColor=<rgba 0.9 0.7119 0.0938 1>
    R1 SkyBlue:LinearColor=<rgba 0.0643 0.3785 0.5365 1>
    R1 TitaniumWhite:LinearColor=<rgba 0.7031 0.7031 0.7031 1>
    R2 Black:LinearColor=<rgba 0.0044 0.0056 0.0116 1>
    R2 BurntSienna:LinearColor=<rgba 0.0595 0.013 0.0086 1>
    R2 Cobalt:LinearColor=<rgba 0.0006 0.0037 0.4179 1>
    R2 Crimson:LinearColor=<rgba 0.1927 0 0.008 1>
    R2 DefaultColor:LinearColor=<rgba 0.0103 0.0307 0.1979 1>
    R2 ForestGreen:LinearColor=<rgba 0 0.099 0.0046 1>
    R2 Grey:LinearColor=<rgba 0.0891 0.0956 0.1042 1>
    R2 Lime:LinearColor=<rgba 0.0662 0.125 0.005 1>
    R2 Orange:LinearColor=<rgba 0.2188 0.0745 0.0031 1>
    R2 Pink:LinearColor=<rgba 0.1875 0.0077 0.1129 1>
    R2 Purple:LinearColor=<rgba 0.1232 0.0032 0.2609 1>
    R2 Saffron:LinearColor=<rgba 0.2135 0.1612 0.0151 1>
    R2 SkyBlue:LinearColor=<rgba 0.0003 0.1221 0.1714 1>
    R2 TitaniumWhite:LinearColor=<rgba 0.2373 0.2338 0.2656 1>
    Saffron=<rgba 0.4792 0.3674 0 1>
    SkyBlue=<rgba 0 0.2634 0.3958 1>
    TitaniumWhite=<rgba 0.224 0.224 0.224 1>
    Use R1 Painted Overrides=true
    Use R2 Painted Overrides=true
   )
   &InitializeParticle(
    Color=$Particles.Color
    Lifetime=$Emitter.CurrentLoopDuration
    Mesh Scale Mode=::NewEnumerator4
    Sprite Rotation Mode=::NewEnumerator2
    UsePositionOffset=false
    Write Material Random=true
   )
   &PartitionParticles()
   &SkeletalMeshLocation(
    Bone / Socket Index=&Add_Integer(A=$Particles.PartitionParticles.Partition, B=4)
    Bone / Socket Sampling Mode=::NewEnumerator5
    Interpolate Mesh Sampling=false
    MaskSpawnGroup=false
    Mesh Sampling Type=::NewEnumerator6
    Skeletal Mesh=$System.SkeletalMesh
   )
   /SetVariables(
    Particles.Stars_Size:Vector2f=&RandomRangeVector2D(Minimum=<5 5>, Maximum=<11 11>)
    Particles.WW2_Color:LinearColor=<rgba 0.0103 0.0307 0.1979 1>
    Particles.WW2_Scale size:Vector3f=<1.1 1.1 1.1>
    Particles.WW3_Color:LinearColor=<rgba 0.2448 0.1034 0.0277 0.913>
    Particles.WW3_LBColor:LinearColor=<rgba 0.1273 0.6251 0.7552 1>
   ) {enabled=false, isSetParameters=true}
  ]
  particleUpdateScript [
   &VC_PaintedVariantsOverride(DefaultColor=&ColorFromCurve(ColorCurve=DataInterface() {propertyValues {
       AlphaCurve {keys [<at 0 4 RCIM_Cubic>,<at 0.9208 0.9976>,<at 1 0 RCIM_Cubic>]}
       BlueCurve {keys [<at 0.0334 0.0196>,<at 0.1929 0.552>,<at 0.4192 0.3093>,<at 1 0.0987>]}
       GreenCurve {keys [<at 0.0334 0.6443>,<at 0.1929 0.0931>,<at 0.4192 0.0562>,<at 1 0.5624>]}
       RedCurve {keys [<at 0.0334 0.956>,<at 0.1929 0.0437>,<at 0.4192 0.0249>,<at 1 0.8204>]}
      }})) {enabled=false, module=VC_PaintedVariantsOverride001}
   &FN_DisableTireFX(
    Mask Index Attribute:NiagaraInt32=$Particles.PartitionParticles.Partition
    Mask Index FL=2
    Mask Index FR=3
    Mask Index RL=0
    Mask Index RR=1
   )
   &VC_WheelWrapScale(Wheel Index Attribute:NiagaraInt32=&Add_Integer(A=$Particles.PartitionParticles.Partition, B=2))
   &DynamicMaterialParameters(
    Index 0 Param 1=3
    Index 0 Param 2=1
    Index 0 Param 3=2.5
    Index 0 Param 4=20
    Module.Index0Param0Write=true
    Module.Index0Param1Write=false
    Module.Index0Param2Write=false
    Module.Index0Param3Write=true
   ) {enabled=false}
   &ParticleState(Kill Particles When Lifetime Has Elapsed=false, Loop Particles Lifetime=true)
   &UpdateMeshOrientation(Rotation Coordinate Space=::NewEnumerator2, Rotation Vector=<0 1 0>) {enabled=false}
   &SkeletalMeshLocation(
    Bone / Socket Index=&Add_Integer(A=$Particles.PartitionParticles.Partition, B=4)
    Bone / Socket Sampling Mode=::NewEnumerator5
    MaskSpawnGroup=false
    Mesh Sampling Type=::NewEnumerator6
    Skeletal Mesh=$System.SkeletalMesh
    Spawn Group Mask=2
   ) {module=SkeletalMeshLocation001}
   &SolveForcesAndVelocity(Clamp Velocity=false, Limit Acceleration=false)
  ]
  rendererClasses [&NiagaraMeshRendererProperties]
  renderers [
   {rendererClass=&NiagaraMeshRendererProperties, rendererIndex=0}
   {rendererClass=&NiagaraMeshRendererProperties, rendererIndex=1}
   {rendererClass=&NiagaraMeshRendererProperties, rendererIndex=2}
  ]
 }
}
summary {
 systemName=NS_DriftTrail_StarryNight
 emitters [
  {emitterName=WheelWrap02, rendererClasses [&NiagaraMeshRendererProperties]}
  {emitterName=NE_VC_DriftTrail_Sprite_Parent, rendererClasses [&NiagaraSpriteRendererProperties]}
  {emitterName=NE_VC_DriftTrail_Ribbon_Parent002, rendererClasses [&NiagaraRibbonRendererProperties]}
 ]
 userVariables [
  User.bIsFrontendPreview {type=NiagaraBool}
  User.bIsPainted {type=NiagaraBool}
  User.Brightness2 {type=NiagaraFloat, defaultValue=1, description=""}
  User.ColorVariantID {type=NiagaraInt32}
  User.ConeColor {type=LinearColor, description="", defaultValue {struct=LinearColor, value=<rgba 1 1 1 1>}}
  User.DTCarOffset {type=NiagaraPosition, description="", defaultValue {}}
  User.DTCarOffset002 {type=NiagaraPosition, description="", defaultValue {}}
  User.GameModeID {type=NiagaraInt32}
  User.MultiVariantColorCone {type=NiagaraFloat, defaultValue=2.5, description=""}
  User.offset {type=NiagaraPosition, description="", defaultValue {}}
  User.offset001 {type=Vector3f, description="", defaultValue {struct=Vector3f, value=<0 0 0>}}
  User.offset002 {type=Vector3f, description="", defaultValue {struct=Vector3f, value=<0 0 0>}}
  User.PaintedColor {type=LinearColor, description="", defaultValue {struct=LinearColor, value=<rgba 1 1 1 1>}}
  User.PaintedColorTriangles {type=LinearColor, description="", defaultValue {struct=LinearColor, value=<rgba 1 1 1 1>}}
  User.PaintedScalar {type=NiagaraFloat, defaultValue=1, description=""}
  User.PowerTriColorPaintedCone {type=NiagaraFloat, defaultValue=1.5, description=""}
  User.VehicleFlags {type=NiagaraInt32, defaultValue=806289408}
  User.Wheel_Radius_Array {
   type=NiagaraDataInterfaceArrayFloat
   defaultValue {
    struct=NiagaraExt_VariableValue_DataInterface
    value {
     dataInterface=&self:NiagaraDataInterfaceArrayFloat_0
     dataInterfaceClass=&NiagaraDataInterfaceArrayFloat
    }
   }
  }
  User.Wheel_Width_Array {
   type=NiagaraDataInterfaceArrayFloat
   defaultValue {
    struct=NiagaraExt_VariableValue_DataInterface
    value {
     dataInterface=&self:NiagaraDataInterfaceArrayFloat_1
     dataInterfaceClass=&NiagaraDataInterfaceArrayFloat
    }
   }
  }
  User.WheelSize {type=NiagaraFloat, defaultValue=0.1, description=""}
 ]
}
